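# Season 30 story conversation for Phase 06 Quest 04, where the Underground Rebel NPC (Jones) urges the player to help Hope avert the Island's doom.
# One entry tag leads into a linear five-line speech chain advanced by Continue choices.
# The first line is an auto-selected main-questline choice, gated on the NPC's conversation tag, the active quest and an incomplete first objective.
# The graph itself contains no objective-completion node.
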ConversationDatabase /BattlePassS30_Quests/Conversations/Story/P06/Quest_04/Conversation_S30_Story_SP_P06_Q04
entry_points [{tag=Conversation.Entry.Season30.Story.P06.Q04.UR, destinations [#1]}]
nodes [
 ConversationEntryPointNode_0 {
  id=#1
  class=ConversationEntryPointNode
  outputs [#2]
  properties {entryTag {tagName=Conversation.Entry.Season30.Story.P06.Q04.UR}}
 }
 Speech_C_5 {
  id=#2
  class=Speech_C
  outputs [#3]
  properties {
   softTaskIcon=&T_UI_NPC_MainQuest
   generalConfig {defaultMessage="The Island's been doomed before. By a mystical box-shaped villain, even!"}
  }
  sub_nodes [
   ConversationChoiceNode_0 {
    class=ConversationChoiceNode
    properties {
     defaultChoiceDisplayText=Talk to Jones
     choiceTags {gameplayTags [
       {tagName=Conversation.Choice.AutoSelectIfValid}
       {tagName=Conversation.Choice.Category.MainQuestline}
       {tagName=Athena.Quests.S30.Conversation.Story.P06.04.UR}
      ]}
    }
   }
   FortConversationRequirement_ControllerRequirement_0 {
    class=FortConversationRequirement_ControllerRequirement
    properties {
     nodeName=Require: NPC Has Conversation Tag
     participantId {tagName=Conversation.Participant.NPC}
     requirement {class=FortControllerRequirement_NPCHasConversationEntryTag, properties {queryTag {tagName=Conversation.Entry.UndergroundRebel}}}
    }
   }
   FortConversationRequirement_ControllerRequirement_1 {
    class=FortConversationRequirement_ControllerRequirement
    properties {
     nodeName=Require: Is Quest Active
     requirement {
      class=FortControllerRequirement_IsQuestActive
      properties {questDefinition=/BattlePassS30_Quests/Items/QuestItems/Story/P06/Quest_S30_Story_P06_Q04}
     }
    }
   }
   FortConversationRequirement_ControllerRequirement_2 {
    class=FortConversationRequirement_ControllerRequirement
    properties {
     nodeName=Require: Has Completed Quest Objective
     requirement {
      class=FortControllerRequirement_HasCompletedQuestObjective
      properties {
       bInverseCondition=true
       objectiveBackendName=Quest_S30_Story_P06_Q04_obj0
       questDefinition=&Quest_S30_Story_P06_Q04
      }
     }
    }
   }
  ]
 }
 Speech_C_2 {
  id=#3
  class=Speech_C
  outputs [#4]
  properties {generalConfig {defaultMessage="But this time feels different. I may not speak, uh, gem-lish like Hope does..."}}
  sub_nodes [ConversationChoiceNode_3 {class=ConversationChoiceNode, properties {defaultChoiceDisplayText=Continue}}]
 }
 Speech_C_0 {
  id=#4
  class=Speech_C
  outputs [#5]
  properties {generalConfig {defaultMessage=...But I've been around enough apocalypses to know who's got the heart to stop one in its tracks.}}
  sub_nodes [ConversationChoiceNode_6 {class=ConversationChoiceNode, properties {defaultChoiceDisplayText=Continue}}]
 }
 Speech_C_3 {
  id=#5
  class=Speech_C
  outputs [#6]
  properties {generalConfig {defaultMessage="This one, I think, is all on Hope."}}
  sub_nodes [ConversationChoiceNode_5 {class=ConversationChoiceNode, properties {defaultChoiceDisplayText=Continue}}]
 }
 Speech_C_4 {
  id=#6
  class=Speech_C
  outputs []
  properties {generalConfig {defaultMessage="We should do what we can to help her. "}}
  sub_nodes [ConversationChoiceNode_4 {class=ConversationChoiceNode, properties {defaultChoiceDisplayText=Continue}}]
 }
]
